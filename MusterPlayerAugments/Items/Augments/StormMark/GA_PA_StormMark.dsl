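; Passive Storm Mark player augment ability that calls MarkinRadius on storm phases, gated by hotfix toggles for storm forming and shrinking.
; On activation it copies duration and radius hotfix values, then listens to stateful game phase events from both standard and Creative storm shield sources through the event router.
; Phase changes fire the augment cue, and marking adds a local burst cue above the avatar carrying the radius.
; Ending stops both listeners.
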(event Custom|PlayClientVFX
  ;[0.3704 0.02034 1] Radius FX
  (GameplayCue|ExecuteGameplayCue%28Burst%29Local (Ability|GetAvatarActorfromActorInfo) (Variables|Default|GetCueOnMark) (Ability|GameplayCue|MakeGameplayCueParameters :NormalizedMagnitude 0.0 :RawMagnitude (ScalableFloat|GetValueatLevel (Variables|Default|GetHotFixRowOverlapRadius)) :Location (+ (Transformation|GetActorLocation (Ability|GetAvatarActorfromActorInfo)) "0,0,500.000000") :Normal "0,0,0" :SourceObject (Ability|GetAvatarActorfromActorInfo) :GameplayEffectLevel 1 :AbilityLevel 1 :bReplicateLocationWhenUsingMinimalRepProxy false)))

(event Ability|EventFailedToActivatePassiveAbility
  (params (ActorInfo :type Gameplay%20Ability%20Actor%20Info%20Structure)))

(event Ability|EventActivateAbility
  (|CallParentFunction)
  (Variables|Default|SetRow_MarkingDuration (Variables|Default|GetHotFixRowMarkingDuration))
  (Variables|Default|SetRow_OverlapRadius (Variables|Default|GetHotFixRowOverlapRadius))
  (bind _returnvalue (GameplayEventRouter|GetEventRouter "/Script/FortniteGame.FortGameState" (Ability|GetOwningActorfromActorInfo)))
  (bind (_asynctaskproxy _payload) (GameplayEventRouter|StartListeningToStatefulEvent _returnvalue "/Script/FortniteGame.FortGamePhaseStepUpdatedEvent")
    (:then
      (Variables|Default|SetGamePhaseStepUpdatedAsyncTask _asynctaskproxy)
      (bind (_asynctaskproxy_2 _payload_2) (GameplayEventRouter|StartListeningToStatefulEvent _returnvalue "/Script/FortniteGame.FortCreativeStormShieldGamePhaseStepUpdatedEvent")
        (:then
          (Variables|Default|SetCreativeStormShieldGamePhaseStepUpdatedAsyncTask _asynctaskproxy_2))
        (:OnEventReceived
          _)
        (:OnSavedEventStateExists
          _))
      (CallFunction|OnGamePhaseStepChanged (.GamePhaseStep (Utilities|Struct|BreakFortCreativeStormShieldGamePhaseStepUpdatedEvent _payload_2))))
    (:OnEventReceived
      _)
    (:OnSavedEventStateExists
      _))
  (CallFunction|OnGamePhaseStepChanged (Utilities|Struct|BreakFortGamePhaseStepUpdatedEvent _payload)))

(event Custom|OnGamePhaseStepChanged
  (params (GamePhaseStep :type EAthenaGamePhaseStep%20Enum))
  ;[1 1 1] TODO: hook up hotfix floats to csv
  (Utilities|ExecuteAugmentGameplayCue (Variables|Default|GetCue%20Tag))
  (if (ScalableFloat|GetValueasBool (Variables|Default|GetHotFixMarkOnStormForm))
    (if (Utilities|Enum|Equal(Enum) GamePhaseStep (Variables|Default|GetStormHoldingMarkingStep))
      _
      (elif (Utilities|Enum|Equal(Enum) GamePhaseStep (Variables|Default|GetStormShrinkingMarkingStep))
        _))
    (CallFunction|MarkinRadius)
    (CallFunction|PlayClientVFX)
    (elif (ScalableFloat|GetValueasBool (Variables|Default|GetHotFixMarkOnStormShrink))
      _)))

(event Ability|EventOnEndAbility
  (params (bWasCancelled :type Boolean))
  (|CallParentFunction bWasCancelled)
  (bind _gamephasestepupdatedasynctask (Variables|Default|GetGamePhaseStepUpdatedAsyncTask)
    (:then
      (GameplayEventRouter|StopListeningtoStatefulEvent _gamephasestepupdatedasynctask))
    (:else
      _))
  (bind _creativestormshieldgamephasestepupdatedasynctask (Variables|Default|GetCreativeStormShieldGamePhaseStepUpdatedAsyncTask)
    (:then
      (GameplayEventRouter|StopListeningtoStatefulEvent _creativestormshieldgamephasestepupdatedasynctask))
    (:else)))
